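# Lit translucent master material for swirling smoke on Niagara and Cascade sprites, using volumetric per-vertex directional lighting.
# Particle color is tinted by morning, day, evening and night parameters through Color_Time_of_Day, with a Use Emissive blend feeding the emissive adjustment function.
# Opacity multiplies time-of-day-scaled particle alpha by an 8x8 smoke flipbook alpha.
# Static switches add depth fade, bypassed on mobile, and near-camera fade.
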Material /Game/Effects/Niagara/MasterMaterials/Smoke/M_FX_Generic_SwirlSmoke_Lit_01
graph {
 properties {
  bUsedWithNiagaraSprites=true
  bUsedWithParticleSprites=true
  blendMode=BLEND_Translucent
  pixelDepthOffsetMode=PDOM_Legacy
  translucencyLightingMode=TLM_VolumetricPerVertexDirectional
  usedShadingModels=Default Lit
 }
 nodes [
  n1 = ParticleColor()
  n2 = Multiply(A=VectorParameter() {parameterName=Color_Morning, defaultValue=<rgba 1 0.5451 0.4 0>}.RGB, B=@n1.RGB)
  n3 = Multiply(A=@n1.RGB, B=VectorParameter() {parameterName=Color_Day, defaultValue=<rgba 1 1 1 1>}.RGB)
  n4 = Multiply(A=@n1.RGB, B=VectorParameter() {parameterName=Color_Evening, defaultValue=<rgba 1 0.4263 0.4 0>}.RGB)
  n5 = Multiply(A=@n1.RGB, B=VectorParameter() {parameterName=Color_Night, defaultValue=<rgba 0.4 0.5658 1 0>}.RGB)
  MaterialFunctionCall(Emissive=LinearInterpolate(
    B=MaterialFunctionCall(Color_Morning=@n2, Color_Day=@n3, Color_Evening=@n4, Color_Night=@n5) {materialFunction=&Color_Time_of_Day}.Result
    Alpha=ScalarParameter() {parameterName=Use Emissive}
   )) {materialFunction=&MF_FX_EmissiveAdjustments}
  n6 = ScalarParameter() {defaultValue=1.5, parameterName=Alpha_Morning}
  n7 = ScalarParameter() {defaultValue=1, parameterName=Alpha_Day}
  n8 = ScalarParameter() {defaultValue=2, parameterName=Alpha_Evening}
  n9 = ScalarParameter() {defaultValue=2, parameterName=Alpha_Night}
  n10 = Multiply(
   A=MaterialFunctionCall(
    Color_Morning=Multiply(A=@n6, B=@n1.A)
    Color_Day=Multiply(A=@n7, B=@n1.A)
    Color_Evening=Multiply(A=@n8, B=@n1.A)
    Color_Night=Multiply(A=@n1.A, B=@n9)
   ) {materialFunction=&Color_Time_of_Day}.Result
   B=ParticleSubUV!(UVs, Tex, Apply View MipBias) {texture=/Game/Effects/Fort_Effects/BluePrints/BeeHive/T_SmokeFlip_03_8x8}.A
  )
  n11 = DepthFade(Opacity=@n10, FadeDistance=ScalarParameter() {defaultValue=32, parameterName=Depth Fade}) {fadeDistanceDefault=32}
  n12 = ScalarParameter() {defaultValue=250, parameterName="Distance (S)"}
  Multiply(
   A=StaticSwitchParameter(True=ShadingPathSwitch(Default=@n11, Mobile=@n10), False=@n10) {parameterName=UseDepthFade?}
   B=StaticSwitchParameter(
    True=MaterialFunctionCall(Distance=@n12) {materialFunction=&Near_Camera_Fade}.Result
    False=Constant() {r=1}
   ) {parameterName=UseNearFade}
  )
 ]
}
