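# Translucent unlit master material for icy smoke or mist particles on sprites and meshes.
# Emissive takes a SubUV flipbook channel, adds panning world-space screen-aligned dust driven by dynamic parameters, then tints by particle color and EmissiveTweak.
# Use_Dust works inverted: 0 adds dust, 1 removes it.
# Opacity multiplies a separately sampled mask, particle alpha and a tiled soft-smoke mask, with near-camera and depth fades.
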Material /Game/Effects/Niagara/MasterMaterials/Smoke/M_IceSmoke_Simple
graph {
 properties {
  bUsedWithMeshParticles=true
  bUsedWithNanite=true
  bUsedWithNiagaraSprites=true
  bUsedWithParticleSprites=true
  blendMode=BLEND_Translucent
  pixelDepthOffsetMode=PDOM_Legacy
  shadingModel=MSM_Unlit
  translucencyLightingMode=TLM_VolumetricPerVertexNonDirectional
  usedShadingModels=Unlit
 }
 nodes [
  n1 = ParticleColor()
  n2 = TextureSampleParameter2D() {parameterName=MaskTex, samplerType=SAMPLERTYPE_LinearColor, texture=/Game/Effects/Fort_Effects/Textures/Smoke/T_FallingMist_8x8_01}
  n3 = Power(Base=ChannelMaskParameter(None=@n2.RGB) {maskChannel=Blue, parameterName=MaskTex, defaultValue=<rgba 0 0 1 0>}.R) {constExponent=1.5}
  n4 = Multiply(A=@n1.A, B=Saturate(None=Multiply(A=@n3) {constB=0.5}))
  n5 = TextureCoordinate() {uTiling=8, vTiling=8}
  n6 = LinearInterpolate(Alpha=TextureSampleParameter2D(UVs=Panner(Coordinate=@n5)) {parameterName=Mask02, texture=/Game/Effects/Fort_Effects/Textures/T_Smoke_Soft}.RGB) {constB=0.8}
  n7 = ScalarParameter() {defaultValue=500, group=Opacity, parameterName=Near Camera Fade Distance, sortPriority=2}
  DepthFade(
   Opacity=Multiply(
    A=Multiply(A=@n4, B=@n6)
    B=MaterialFunctionCall(Distance=@n7) {materialFunction=&Near_Camera_Fade}.Result
   )
   FadeDistance=ScalarParameter() {defaultValue=20, group=Opacity, parameterName=Depth Fade, sortPriority=1}
  )
  n8 = TextureSampleParameterSubUV!(UVs, Bias, Apply View MipBias) {
   group=Texture
   mipValueMode=TMVM_MipBias
   parameterName=BaseTex
   samplerType=SAMPLERTYPE_LinearColor
   texture=&T_FallingMist_8x8_01
  }
  n9 = Power(Base=ChannelMaskParameter(None=@n8.RGB) {maskChannel=Blue, parameterName=BaseTex, defaultValue=<rgba 0 0 1 0>}.R) {constExponent=1}
  n10 = LinearInterpolate(A=ScalarParameter() {parameterName=SubUVMinBrightness}, Alpha=Saturate(None=Multiply(A=@n9)))
  n11 = MaterialFunctionCall!(Center Point World Position) {materialFunction=/Engine/Functions/Engine_MaterialFunctions02/UVs/WorldSpaceAlignedScreenCoordinates}
  n12 = DynamicParameter() {defaultValue=<rgba 1 1 1 0>, paramNames [DustTiling, Dust Emissive, Cloud Density, Panning MacroUV Control]}
  n13 = Multiply(
   A=Multiply(A=ScalarParameter() {defaultValue=2, parameterName=DustTilingU}, B=@n12.DustTiling)
   B=Multiply(A=ScalarParameter() {defaultValue=2, parameterName=DustTilingV}, B=@n12.DustTiling)
  )
  n14 = Time()
  n15 = ScalarParameter() {group=Texture, parameterName=U_Speed, sortPriority=5}
  n16 = ScalarParameter() {defaultValue=1, group=Texture, parameterName=V_Speed, sortPriority=6}
  n17 = TextureSampleParameter2D(UVs=Panner(
    Coordinate=Multiply(A=@n11.Offset and Scaled UVS, B=@n13)
    Time=Multiply(A=@n12.Panning MacroUV Control, B=@n14)
    Speed=AppendVector(A=@n15, B=@n16)
   )) {group=Texture, parameterName=Tiling Dust Texture, sortPriority=2, texture=/Game/Global/Textures/Default/PointSplatter/T_Tiling_Dust}
  n18 = ScalarParameter() {defaultValue=2, parameterName=Tiling Dust Texture Light}
  n19 = LinearInterpolate(A=Add(A=@n10, B=Multiply(A=@n17.RGB, B=@n18) {constB=2}), B=@n10, Alpha=ScalarParameter() {parameterName=Use_Dust})
  MaterialFunctionCall(Emissive=Multiply(
    A=Multiply(A=@n19, B=@n1.RGB)
    B=VectorParameter() {group=Color, parameterName=EmissiveTweak, sortPriority=1, defaultValue=<rgba 1 1 1 0>}.RGB
   ) {constB=0.5}) {materialFunction=&MF_FX_EmissiveAdjustments}
 ]
}
