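# Translucent, lit Niagara sprite master material for raymarched smoke, built around MF_Raymarched_Smoke_Func.
# The function's emissive is desaturated by a parameter and passed through MF_FX_EmissiveAdjustments, alongside particle color dimmed by a constant multiply.
# A static switch optionally cuts opacity inside a world-space sphere mask, with radius and center from dynamic parameter 1.
# The mask center feeds that parameter's X channel into all three axes, leaving Y and Z unused.
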Material /Game/Effects/Niagara/MasterMaterials/Smoke/M_RayMarched_Smoke
graph {
 properties {bAutomaticallySetUsageInEditor=false, bUsedWithNiagaraSprites=true, blendMode=BLEND_Translucent, usedShadingModels=Default Lit}
 nodes [
  n1 = MaterialFunctionCall() {materialFunction=/Game/Effects/Raymarch/MF_Raymarched_Smoke_Func}
  MaterialFunctionCall(Emissive=Desaturation(None=@n1.Emissive, Fraction=ScalarParameter() {parameterName=Des})) {materialFunction=&MF_FX_EmissiveAdjustments}
  Multiply(A=ParticleColor().RGB) {constB=0.2}
  n2 = DynamicParameter() {parameterIndex=1, defaultValue=<rgba 1 1 1 300>, paramNames [X,Y,Z,Radius]}
  n3 = SphereMask(
   A=WorldPosition().XYZ
   B=MaterialFunctionCall(X=@n2.X, Y=@n2.X, Z=@n2.X) {materialFunction=&MakeFloat3}.Result
   Radius=@n2.Radius
   Hardness=ScalarParameter() {defaultValue=0.55, parameterName=Hardness}
  ) {hardnessPercent=50}
  StaticSwitchParameter(True=Multiply(A=@n1.Opacity, B=OneMinus(None=@n3)), False=@n1.Opacity) {parameterName=UseCenterMask}
 ]
}
