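# Translucent unlit Niagara sprite material that draws a round, noisy smoke puff tinted and faded by particle color.
# A panning noise texture distorts the UVs of layered noise samples, which modulate a radial gradient sized by the Radius parameter.
# The result indexes a Dark Stance glider curve atlas row, whose red channel is thinned by an off-center soft sphere mask to shape opacity.
# Particle RGB passes through a shared emissive adjustment function.
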Material /Game/Effects/Niagara/MasterMaterials/Smoke/M_Round_Sprite_Smoke
graph {
 properties {bUsedWithNiagaraSprites=true, blendMode=BLEND_Translucent, pixelDepthOffsetMode=PDOM_Legacy, shadingModel=MSM_Unlit, usedShadingModels=Unlit}
 nodes [
  Constant()
  n1 = ParticleColor()
  MaterialFunctionCall(Emissive=@n1.RGB) {materialFunction=&MF_FX_EmissiveAdjustments}
  n2 = Add(A=TextureCoordinate()) {constB=0}
  n3 = ScalarParameter() {defaultValue=0.55, parameterName=Radius}
  n4 = Add(A=TextureCoordinate() {uTiling=0.3, vTiling=0.3}, B=Constant2Vector() {r=0.3})
  n5 = TextureSample(UVs=Panner(Coordinate=TextureCoordinate() {uTiling=0.3, vTiling=0.3}) {speedX=0.5}) {
   samplerType=SAMPLERTYPE_LinearColor
   texture=/Engine/Functions/Engine_MaterialFunctions02/ExampleContent/Textures/OffsetNoiseDistanceFields_ForNormals
  }
  n6 = Multiply(A=@n5.R) {constB=0.07}
  n7 = TextureCoordinate()
  n8 = Constant2Vector()
  n9 = Multiply(A=@n5.R) {constB=0.2}
  n10 = TextureSample(UVs=Add(A=Add(A=@n7, B=@n8), B=@n9)) {texture=/Game/Effects/Fort_Effects/Textures/Tile/T_Noise_02}
  n11 = TextureCoordinate()
  n12 = Constant2Vector() {g=0.3, r=0.1}
  n13 = TextureSample(UVs=Add(A=Add(A=@n11, B=@n12), B=@n9)) {texture=&T_Noise_02}
  n14 = Max(
   A=TextureSample(UVs=Add(A=@n4, B=@n6)) {samplerType=SAMPLERTYPE_Masks, texture=/Game/Weapons/WeaponSkins/Wraps/Textures/Mash/Mash_Noise}.R
   B=Multiply(A=Add(A=@n10.R, B=@n13.R))
  )
  n15 = CurveAtlasRowParameter(CurveTime=Multiply(
    A=MaterialFunctionCall(UVs=@n2, Radius=@n3) {materialFunction=&RadialGradientExponential}.RadialGradientExponential
    B=Add(B=@n14)
   )) {
   atlas=/Game/Gadgets/Assets/VinderTech_GliderChute/Glider_DarkStance/FX/CAtlas_FX_DarkStance
   curve=/Game/Gadgets/Assets/VinderTech_GliderChute/Glider_DarkStance/FX/CV_DarkStance_Glider_Smoke_Color
   parameterName=Contract
  }
  n16 = SphereMask(A=TextureCoordinate(), B=Constant2Vector() {g=0.6, r=0.45}) {attenuationRadius=0.5, hardnessPercent=0}
  n17 = Multiply(A=Add(B=@n15.R), B=LinearInterpolate(Alpha=OneMinus(None=@n16)) {constA=0.1})
  Multiply(A=@n1.A, B=Saturate(None=Multiply(A=@n17) {constB=2}))
 ]
}
