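# Translucent master material for soft, panning smoke trails on Niagara ribbons and mesh particles, with SplineThicken world position offset for width.
# Dynamic parameters drive texture erosion, pan speed, emissive and master alpha, with pan time offset by actor position.
# Opacity combines tail gradient, depth, camera and view-angle fades, simplified on mobile, plus an optional world sphere mask.
# Particle color, optionally textured, becomes HDR emissive, hue-shifted and brightened by a radial UV mask.
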Material /Game/Effects/Niagara/MasterMaterials/Smoke/M_Smoke_Trail_Soft_Spline
graph {
 properties {
  bUsedWithNiagaraMeshParticles=true
  bUsedWithNiagaraRibbons=true
  blendMode=BLEND_Translucent
  pixelDepthOffsetMode=PDOM_Legacy
  usedShadingModels=Default Lit
 }
 nodes [
  DynamicParameter() {parameterIndex=1, defaultValue=<rgba 1 0 1 1>, paramNames ["", "", "", ""]}
  n1 = ParticleColor()
  n2 = DynamicParameter() {defaultValue=<rgba 1 0 1 1>, paramNames [Erode, EmissiveAmount, PanSpeed, MasterAlpha]}
  n3 = Time()
  n4 = MaterialFunctionCall(Float2=ActorPositionWS()) {materialFunction=&BreakOutFloat2Components}
  n5 = ScalarParameter() {defaultValue=1, parameterName=OffsetModulus, sliderMax=1}
  n6 = Panner(
   Coordinate=NamedRerouteUsage()
   Time=Add(A=Multiply(A=@n2.PanSpeed, B=@n3), B=Modulo!(A=@n4.G, B=@n5))
   Speed=VectorParameter() {parameterName=PanSpeed, defaultValue=<rgba 0.6 0 0 1>}.RGB
  ) {speedX=0.6}
  n7 = TextureSampleParameter2D(UVs=@n6) {parameterName=TrailTexture_BChannel, samplerType=SAMPLERTYPE_Grayscale, texture=&T_Smoke_Trail_Soft}
  n8 = Add(A=@n2.Erode, B=@n7.G)
  n9 = Clamp(None=Multiply(A=LinearInterpolate(Alpha=@n8) {constA=-3}, B=@n7.G))
  n10 = Multiply(A=@n1.A, B=@n9)
  n11 = Multiply(
   A=MaterialFunctionCall(
    Fade Length=ScalarParameter() {defaultValue=350, group=Camera Fade, parameterName=Near Fade Length}
    Fade Offset=ScalarParameter() {defaultValue=24, group=Camera Fade, parameterName=Near Fade Offset}
   ) {materialFunction=&CameraDepthFade}.Result
   B=NamedRerouteUsage()
  )
  n12 = MaterialFunctionCall() {materialFunction=&LinearGradient}
  n13 = Multiply(A=OneMinus(None=@n12.UGradient), B=@n12.UGradient)
  n14 = Saturate(None=MaterialFunctionCall(In=Multiply(A=@n13) {constB=8}, Contrast=ScalarParameter() {defaultValue=0.1, parameterName=TailGradientFallOff}) {materialFunction=&CheapContrast}.Result)
  n15 = DynamicParameter() {desc=Index 1 Params for Niagara, parameterIndex=1, defaultValue=<rgba 1 0 1 1>, paramNames ["", "", "", ""]}
  n16 = ScalarParameter() {defaultValue=40, parameterName=DepthFade}
  n17 = DepthFade(
   Opacity=Multiply(A=Multiply(A=@n10, B=@n11), B=@n14)
   FadeDistance=If(A=@n15.R, B=Constant() {r=1}, "A > B"=@n15.R, "A == B"=@n16, "A < B"=@n16)
  ) {fadeDistanceDefault=40}
  n18 = MaterialFunctionCall("FadeSensitivity (optional)"=ScalarParameter() {parameterName=EdgeFadeSensitivity}) {materialFunction=&FadeBasedOnViewAngle}
  n19 = ShadingPathSwitch(
   Default=Multiply(
    A=Multiply(A=@n17, B=@n18.Result)
    B=DynamicParameter() {defaultValue=<rgba 1 0 1 1>, paramNames [Erode, EmissiveAmount, PanSpeed, MasterAlpha]}.MasterAlpha
   )
   Mobile=Multiply(A=@n10, B=@n14)
  )
  n20 = ScalarParameter() {group="4.2) Emissive Sphere Mask", parameterName=Sphere Mask Low Value, sortPriority=1}
  n21 = ScalarParameter() {defaultValue=1, group="4.2) Emissive Sphere Mask", parameterName=Sphere Mask High Value, sortPriority=2}
  n22 = ObjectPositionWS()
  n23 = VectorParameter() {group="4.2) Emissive Sphere Mask", parameterName=Location Offset, sortPriority=5, defaultValue=<rgba 0 0 0 1>}
  n24 = Multiply(
   A=SphereMask(
    A=WorldPosition().XYZ
    B=Add(A=@n22, B=@n23.RGB)
    Radius=ScalarParameter() {defaultValue=50, group="4.2) Emissive Sphere Mask", parameterName=Sphere Mask Radius, sortPriority=3}
    Hardness=ScalarParameter() {group="4.2) Emissive Sphere Mask", parameterName=Sphere Mask Hardness, sortPriority=4}
   )
   B=ScalarParameter() {group="4.2) Emissive Sphere Mask", parameterName=Hardness, sortPriority=6}
  )
  n25 = Saturate(None=Power(Base=Clamp(None=@n24), Exp=ScalarParameter() {defaultValue=1, group="4.2) Emissive Sphere Mask", parameterName=Power, sortPriority=5}) {constExponent=1})
  LinearInterpolate(
   A=@n19
   B=Multiply(A=Saturate(None=@n19), B=LinearInterpolate(A=@n20, B=@n21, Alpha=@n25))
   Alpha=ScalarParameter() {group="4.2) Emissive Sphere Mask", parameterName="Use Sphere Mask (World)", sortPriority=-1}
  )
  n26 = Saturate(None=NamedRerouteUsage())
  n27 = OneMinus(None=@n26)
  n28 = Multiply(
   A=Power(
    Base=StaticSwitchParameter(True=@n27, False=@n26) {group=Hue Mask, parameterName=Reverse Hue Mask}
    Exp=ScalarParameter() {defaultValue=4, group=Hue Mask, parameterName=Hue Mask Exponent, sliderMax=1}
   )
   B=ScalarParameter() {defaultValue=-0.1, group=Hue Mask, parameterName=Hue Mask Scale, sliderMax=1}
  )
  n29 = TextureSampleParameter2D(UVs=@n6) {parameterName=ColorTexture, texture=/Game/Effects/Fort_Effects/Textures/Smoke/T_Smoke_Trail_Soft_Chromatic}
  n30 = OneMinus(None=@n29.RGB)
  n31 = ScalarParameter() {parameterName=InvertColorMap}
  n32 = Multiply(
   A=@n1.RGB
   B=StaticSwitchParameter(True=LinearInterpolate(A=@n29.RGB, B=@n30, Alpha=@n31), False=Constant() {r=1}) {parameterName=UseColorTexture}
  )
  n33 = DynamicParameter() {defaultValue=<rgba 1 0 1 1>, paramNames [Erode, EmissiveAmount, PanSpeed, MasterAlpha]}
  n34 = MaterialFunctionCall(LDR Input=Saturate(None=@n32), HDR Tint=Multiply(A=@n32, B=Multiply(A=@n33.EmissiveAmount, B=@n9))) {materialFunction=&DeriveHDRfromLDR}
  n35 = MaterialFunctionCall(
   Input=@n26
   Input Low=ScalarParameter() {defaultValue=1, group=Brightness Mask, parameterName=Brightness Radial Mask Min, sliderMax=1}
   Input High=ScalarParameter() {defaultValue=0.1, group=Brightness Mask, parameterName=Brightness Radial Mask Max, sliderMax=1}
   Target Low=ScalarParameter() {defaultValue=0.3, group=Brightness Mask, parameterName=Brightness Radial Min, sliderMax=1}
   Target High=ScalarParameter() {defaultValue=2.25, group=Brightness Mask, parameterName=Brightness Radial Max, sliderMax=1}
  ) {materialFunction=&RemapValueRange}
  n36 = ScalarParameter() {defaultValue=3, group=Brightness Mask, parameterName=Brightness Radial Exponent, sliderMax=1}
  MaterialFunctionCall(Emissive=Multiply(
    A=MaterialFunctionCall(Hue Shift Percentage=@n28, Texture=@n34.RemappedRange) {materialFunction=&HueShift}.Result
    B=Power(Base=@n35.Result, Exp=@n36)
   )) {materialFunction=&MF_FX_EmissiveAdjustments}
  n37 = ScalarParameter() {group=WPO, parameterName=Spline Width, sliderMax=5, sliderMin=5}
  n38 = WorldPosition()
  n39 = StaticBoolParameter() {group=UV, parameterName=Swap UV}
  NamedRerouteDeclaration(None=Multiply(
    A=MaterialFunctionCall(WidthBase=@n37, WidthTip=@n37, WorldPosition=@n38.XYZ, Expand U or V UV Channel=@n39) {materialFunction=&SplineThicken}.WorldPositionOffset
    B=NamedRerouteUsage()
   )) {name=WPO, nodeColor=<rgba 1 0 0.8672 1>}
  n40 = MaterialFunctionCall(Float2=TextureCoordinate() {coordinateIndex=1}) {materialFunction=&BreakOutFloat2Components}
  n41 = OneMinus(None=@n40.R)
  n42 = OneMinus(None=@n40.G)
  n43 = MaterialFunctionCall(
   X=Power(
    Base=StaticSwitchParameter(True=@n41, False=@n40.R) {group=UV, parameterName=Reverse U}
    Exp=ScalarParameter() {defaultValue=1, group=Mask, parameterName=Mask Exponent U, sliderMax=1}
   ) {constExponent=0.5}
   Y=Power(
    Base=StaticSwitchParameter(True=@n42, False=@n40.G) {group=UV, parameterName=Reverse V}
    Exp=ScalarParameter() {defaultValue=1, group=Mask, parameterName=Mask Exponent V, sliderMax=1}
   ) {constExponent=0.5}
  ) {materialFunction=&MakeFloat2}
  n44 = StaticSwitchParameter(
   True=MaterialFunctionCall(XYZ=@n43.Result) {materialFunction=&Swizzle}.YX
   False=@n43.Result
  ) {group=UV, parameterName=Swap UV}
  n45 = ScalarParameter() {defaultValue=0.5, group=Mask, parameterName=Mask Center U, sliderMax=0.5}
  n46 = ScalarParameter() {defaultValue=0.5, group=Mask, parameterName=Mask Center V, sliderMax=0.5}
  NamedRerouteDeclaration(None=SphereMask(
    A=NamedRerouteDeclaration(None=@n44) {name=UV, nodeColor=<rgba 0.4062 0 1 1>}
    B=AppendVector(A=@n45, B=@n46)
    Radius=ScalarParameter() {defaultValue=0.5, group=Mask, parameterName=Mask Radius}
    Hardness=ScalarParameter() {defaultValue=1, group=Mask, parameterName=Mask Hardness}
   )) {name=Mask, nodeColor=<rgba 1 0.0065 0.1633 1>}
  n47 = Length(None=Subtract(A=CameraPositionWS(), B=WorldPosition().XYZ)) {is=class MaterialExpressionLength}
  n48 = ScalarParameter() {defaultValue=1000, group=Camera Fade, parameterName=Camera Fade Far Start, sliderMax=50000}
  n49 = ScalarParameter() {defaultValue=3000, group=Camera Fade, parameterName=Camera Fade Far End, sliderMax=70000}
  n50 = Constant() {r=1}
  n51 = Constant()
  NamedRerouteDeclaration(None=Saturate(None=MaterialFunctionCall(Input=@n47, Input Low=@n48, Input High=@n49, Target Low=@n50, Target High=@n51) {materialFunction=&RemapValueRange}.Result)) {name=Camera Fade Far, nodeColor=<rgba 1 0.0469 0 1>}
  NamedRerouteUsage()
  Reroute(None=@n1.RGB)
  StaticSwitchParameter() {group=UV, parameterName=Swap UV}
  TextureSample(UVs=TextureCoordinate())
 ]
}
