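# Unlit translucent Niagara sprite material for cheap swirly SubUV smoke, defaulting to an 8x8 inky smoke flipbook.
# Three slightly UV-offset green-channel samples form over-saturated RGB fringing, with a single-texture switch as alternative.
# Particle color takes optional time-of-day tint, and another switch adds a color-curve lookup driven by a sphere mask around the actor.
# Refined texture alpha times particle alpha forms opacity, skipping depth fade on Forward and Mobile paths.
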Material /Game/Effects/Niagara/MasterMaterials/Smoke/M_SwirlySmoke_SubUV_Cheap
graph {
 properties {
  bUseTranslucencyVertexFog=false
  bUsedWithNiagaraSprites=true
  blendMode=BLEND_Translucent
  pixelDepthOffsetMode=PDOM_Legacy
  shadingModel=MSM_Unlit
  usedShadingModels=Unlit
 }
 nodes [
  ComponentMask(None=ParticleMacroUV()) {g=true}
  Frac(None=Divide(A=ParticlePositionWS()) {constB=25})
  n1 = CurveAtlasRowParameter(CurveTime=SphereMask(
    A=WorldPosition() {worldPositionShaderOffset=WPT_CameraRelative}.XYZ
    B=ActorPositionWS()
    Radius=ScalarParameter() {defaultValue=200, parameterName=Radius}
    Hardness=ScalarParameter() {defaultValue=0.25, parameterName=Hardness}
   )) {
   atlas=&AthenaCurveAtlas01
   curve=/Game/Characters/Player/Male/Medium/Bodies/M_MED_BattleSuit/FX/BattleSuit_Green
   defaultValue=85
   parameterName=ColorCurve
  }
  n2 = Constant() {r=1}
  n3 = MaterialFunctionCall() {materialFunction=/Game/Effects/Fort_Effects/Materials/_Functions/CurveColor_Time_of_Day}
  n4 = Multiply(
   A=ScalarParameter() {defaultValue=1, parameterName=TimeofDayAmount}
   B=DynamicParameter() {defaultValue=<rgba 0.5 1 1 1>, paramNames [FlatAlpha, TimeofDayAmount, Param3, Param4]}.TimeofDayAmount
  )
  n5 = Multiply(
   A=StaticSwitchParameter(True=LinearInterpolate(A=@n2, B=@n3.Result, Alpha=@n4), False=@n2) {parameterName=UseTODColor?}
   B=ParticleColor().RGB
  )
  MaterialFunctionCall(Emissive=StaticSwitchParameter(True=Add(A=@n1, B=@n5), False=@n5) {parameterName=UseAddColorWithCurve?}) {materialFunction=&MF_FX_EmissiveAdjustments}
  n6 = TextureCoordinate()
  n7 = Multiply() {constA=1.5, constB=0.001}
  n8 = TextureSampleParameterSubUV!(UVs=Add(A=@n6, B=@n7), Apply View MipBias) {parameterName=Texture_01, samplerType=SAMPLERTYPE_LinearColor, texture=/Game/Effects/Fort_Effects/Textures/Smoke/T_Inky_Smoke_SubUV_8x8}
  n9 = TextureSampleParameterSubUV(UVs=@n6) {parameterName=Texture_02, samplerType=SAMPLERTYPE_LinearColor, texture=&T_Inky_Smoke_SubUV_8x8}
  n10 = TextureSampleParameterSubUV(UVs=Subtract(A=@n6, B=@n7)) {parameterName=Texture_03, samplerType=SAMPLERTYPE_LinearColor, texture=&T_Inky_Smoke_SubUV_8x8}
  n11 = Desaturation(
   None=MaterialFunctionCall(X=@n8.G, Y=@n9.G, Z=@n10.G) {materialFunction=&MakeFloat3}.Result
   Fraction=Constant() {r=-1}
  )
  n12 = VectorParameter() {parameterName=Refine, defaultValue=<rgba 1 1 1 0.5>}
  n13 = LinearInterpolate(
   A=Multiply(A=Power(Base=@n9.G, Exp=@n12.R), B=@n12.G)
   B=Multiply(A=@n9.G, B=@n12.B)
   Alpha=DynamicParameter() {defaultValue=<rgba 0.5 1 1 1>, paramNames [FlatAlpha, TimeofDayAmount, Param3, Param4]}.FlatAlpha
  )
  n14 = Add(A=ScalarParameter() {parameterName=Add}, B=@n13)
  n15 = TextureSampleParameterSubUV(UVs=@n6) {parameterName=Texture_Single, samplerType=SAMPLERTYPE_LinearColor, texture=&T_Inky_Smoke_SubUV_8x8}
  Multiply(A=@n5, B=StaticSwitchParameter(True=@n15.G, False=Multiply(A=@n11, B=@n14)) {parameterName=UseSingleTexture?})
  n16 = Multiply(A=ParticleColor().A, B=StaticSwitchParameter(True=@n15.G, False=@n13) {parameterName=UseSingleTexture?})
  n17 = DepthFade(Opacity=@n16, FadeDistance=ScalarParameter() {defaultValue=20, parameterName=DepthFade})
  ShadingPathSwitch(Default=@n17, Deferred=@n17, Forward=@n16, Mobile=@n16)
 ]
}
